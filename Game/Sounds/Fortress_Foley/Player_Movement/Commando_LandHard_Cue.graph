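# Sound cue for the Commando's hard landing foley in player movement.
# It mixes a random soldier landing step with a quieter random Commando landing foley layer, each drawn with equal weights from a few variations.
# A modulator at the root varies pitch and volume on each play.
# It uses player footstep attenuation and the player foley sound class, with a concurrency override that prevents new instances at the limit.
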SoundCue /Game/Sounds/Fortress_Foley/Player_Movement/Commando_LandHard_Cue
root=Modulator_4
nodes [
 {id=Mixer_2, children [Random_3,Random_4], properties {inputVolume [1,0.5]}}
 {id=Modulator_4, children [Mixer_2], properties {pitchMax=1.1, pitchMin=0.9, volumeMax=1.3, volumeMin=0.7}}
 {id=Random_3, children [WavePlayer_19, WavePlayer_20], properties {weights [1,1]}}
 {id=Random_4, children [WavePlayer_3,WavePlayer_4,WavePlayer_5], properties {weights [1,1,1]}}
 {id=WavePlayer_19, properties {soundWaveAssetPtr=/Game/Sounds/Fort_Step_Sounds/Human/soldier_land_01}}
 {id=WavePlayer_20, properties {soundWaveAssetPtr=/Game/Sounds/Fort_Step_Sounds/Human/soldier_land_02}}
 {id=WavePlayer_3, properties {soundWaveAssetPtr=/Game/Sounds/Fortress_Foley/Commando/Commando_Land_006}}
 {id=WavePlayer_4, properties {soundWaveAssetPtr=/Game/Sounds/Fortress_Foley/Commando/Commando_Land_002}}
 {id=WavePlayer_5, properties {soundWaveAssetPtr=/Game/Sounds/Fortress_Foley/Commando/Commando_Land_005}}
]
properties {
 attenuationSettings=/Game/Sounds/Attenuation/Fort_Player_Footsteps_Att
 bOverrideConcurrency=true
 maxDistance=3500
 soundClassObject=&Fort_Player_Foley
 volumeMultiplier=1
 concurrencyOverrides {resolutionRule=PreventNew}
}
