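# Sound cue for the Commando landing foley in player movement, played through the player foley sound class with footstep attenuation.
# A mixer layers a random pick of three land sounds at full level with a quieter random pick of three stereo foley sounds.
# Each layer has its own pitch and volume modulation, with the foley layer pitched up.
# Overridden concurrency prevents new instances once the limit is reached.
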SoundCue /Game/Sounds/Fortress_Foley/Player_Movement/Commando_Land_Cue
root=Mixer_5
nodes [
 {id=Mixer_5, children [Modulator_6,Modulator_5], properties {inputVolume [0.4,1]}}
 {id=Modulator_5, children [Random_0], properties {pitchMax=1.1, pitchMin=0.9, volumeMax=1.3, volumeMin=0.7}}
 {id=Modulator_6, children [Random_1], properties {pitchMax=1.4, pitchMin=1.1, volumeMax=1, volumeMin=0.5}}
 {id=Random_0, children [WavePlayer_0,WavePlayer_1,WavePlayer_2], properties {weights [1,1,1]}}
 {id=Random_1, children [WavePlayer_338, WavePlayer_339, WavePlayer_341], properties {weights [1,1,1]}}
 {id=WavePlayer_0, properties {soundWaveAssetPtr=/Game/Sounds/Fortress_Foley/Commando/Commando_Land_006}}
 {id=WavePlayer_1, properties {soundWaveAssetPtr=/Game/Sounds/Fortress_Foley/Commando/Commando_Land_002}}
 {id=WavePlayer_2, properties {soundWaveAssetPtr=/Game/Sounds/Fortress_Foley/Commando/Commando_Land_005}}
 {id=WavePlayer_338, properties {soundWaveAssetPtr=/Game/Sounds/Fortress_Foley/Commando/Commando_Land_Foley_Stereo_003}}
 {id=WavePlayer_339, properties {soundWaveAssetPtr=/Game/Sounds/Fortress_Foley/Commando/Commando_Land_Foley_Stereo_005}}
 {id=WavePlayer_341, properties {soundWaveAssetPtr=/Game/Sounds/Fortress_Foley/Commando/Commando_Land_Foley_Stereo_004}}
]
properties {
 attenuationSettings=/Game/Sounds/Attenuation/Fort_Player_Footsteps_Att
 bOverrideAttenuation=true
 bOverrideConcurrency=true
 maxDistance=4000
 soundClassObject=&Fort_Player_Foley
 volumeMultiplier=1.2
 concurrencyOverrides {resolutionRule=PreventNew}
}
