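# Sound cue for Commando running footsteps, layering step impacts and clothing or gear foley under player footstep attenuation and the player foley sound class.
# A mixer blends three equally weighted random pools: Commando steps at full level, soldier run steps at half and stereo foley kept low.
# Each pool has its own pitch and volume modulator for variation.
# A concurrency override caps simultaneous voices, stopping the oldest first.
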SoundCue /Game/Sounds/Fortress_Foley/Player_Movement/Commando_Step_Run_Cue
root=Mixer_2
nodes [
 {id=Mixer_2, children [Modulator_9,Modulator_12,Modulator_8], properties {inputVolume [0.5,0.1,1]}}
 {id=Modulator_12, children [Random_0], properties {pitchMax=1.4, pitchMin=1.1, volumeMax=1, volumeMin=0.3}}
 {id=Modulator_8, children [Random_1], properties {pitchMax=1.6, pitchMin=1.4, volumeMax=1, volumeMin=0.7}}
 {id=Modulator_9, children [Random_2], properties {pitchMax=1.3, pitchMin=1, volumeMax=1, volumeMin=0.7}}
 {id=Random_0, children [WavePlayer_0,WavePlayer_1,WavePlayer_2,WavePlayer_3,WavePlayer_4,WavePlayer_5], properties {weights [1,1,1,1,1,1]}}
 {
  id=Random_1
  children [WavePlayer_6, WavePlayer_7, WavePlayer_8, WavePlayer_9, WavePlayer_10, WavePlayer_11]
  properties {weights [1,1,1,1,1,1]}
 }
 {id=Random_2, children [WavePlayer_78, WavePlayer_79, WavePlayer_80, WavePlayer_81], properties {weights [1,1,1,1]}}
 {id=WavePlayer_0, properties {soundWaveAssetPtr=/Game/Sounds/Fortress_Foley/Commando/Commando_Foley_Stereo_006}}
 {id=WavePlayer_1, properties {soundWaveAssetPtr=/Game/Sounds/Fortress_Foley/Commando/Commando_Foley_Stereo_008}}
 {id=WavePlayer_10, properties {soundWaveAssetPtr=/Game/Sounds/Fortress_Foley/Commando/Commando_Step_006}}
 {id=WavePlayer_11, properties {soundWaveAssetPtr=/Game/Sounds/Fortress_Foley/Commando/Commando_Step_010}}
 {id=WavePlayer_2, properties {soundWaveAssetPtr=/Game/Sounds/Fortress_Foley/Commando/Commando_Foley_Stereo_001}}
 {id=WavePlayer_3, properties {soundWaveAssetPtr=/Game/Sounds/Fortress_Foley/Commando/Commando_Foley_Stereo_003}}
 {id=WavePlayer_4, properties {soundWaveAssetPtr=/Game/Sounds/Fortress_Foley/Commando/Commando_Foley_Stereo_005}}
 {id=WavePlayer_5, properties {soundWaveAssetPtr=/Game/Sounds/Fortress_Foley/Commando/Commando_Foley_Stereo_009}}
 {id=WavePlayer_6, properties {soundWaveAssetPtr=/Game/Sounds/Fortress_Foley/Commando/Commando_Step_001}}
 {id=WavePlayer_7, properties {soundWaveAssetPtr=/Game/Sounds/Fortress_Foley/Commando/Commando_Step_002}}
 {id=WavePlayer_78, properties {soundWaveAssetPtr=/Game/Sounds/Fort_Step_Sounds/Human/soldier_run_step_02}}
 {id=WavePlayer_79, properties {soundWaveAssetPtr=/Game/Sounds/Fort_Step_Sounds/Human/soldier_run_step_03}}
 {id=WavePlayer_8, properties {soundWaveAssetPtr=/Game/Sounds/Fortress_Foley/Commando/Commando_Step_003}}
 {id=WavePlayer_80, properties {soundWaveAssetPtr=/Game/Sounds/Fort_Step_Sounds/Human/soldier_run_step_04}}
 {id=WavePlayer_81, properties {soundWaveAssetPtr=/Game/Sounds/Fort_Step_Sounds/Human/soldier_run_step_05}}
 {id=WavePlayer_9, properties {soundWaveAssetPtr=/Game/Sounds/Fortress_Foley/Commando/Commando_Step_004}}
]
properties {
 attenuationSettings=/Game/Sounds/Attenuation/Fort_Player_Footsteps_Att
 bOverrideConcurrency=true
 maxDistance=3500
 pitchMultiplier=0.85
 priority=0.5
 soundClassObject=&Fort_Player_Foley
 volumeMultiplier=0.6
 concurrencyOverrides {maxCount=4, resolutionRule=StopOldest}
}
