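# Sound cue for Commando walking footsteps, layering gear foley with step impacts.
# A mixer plays two random six-wave pools, each through its own pitch and volume modulator, with the foley layer mixed far below the steps.
# Steps are pitched up, while foley ranges from lowered to original pitch.
# It uses player footstep attenuation, the player foley sound class, and a small stop-oldest concurrency cap.
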SoundCue /Game/Sounds/Fortress_Foley/Player_Movement/Commando_Step_Walk_Cue
root=Mixer_1
nodes [
 {id=Mixer_1, children [Modulator_11,Modulator_2], properties {inputVolume [0.1,1]}}
 {id=Modulator_11, children [Random_3], properties {pitchMax=1, pitchMin=0.7, volumeMax=1, volumeMin=0.3}}
 {id=Modulator_2, children [Random_5], properties {pitchMax=1.5, pitchMin=1.3, volumeMax=1, volumeMin=0.7}}
 {
  id=Random_3
  children [WavePlayer_6, WavePlayer_7, WavePlayer_8, WavePlayer_9, WavePlayer_10, WavePlayer_11]
  properties {weights [1,1,1,1,1,1]}
 }
 {
  id=Random_5
  children [WavePlayer_18, WavePlayer_19, WavePlayer_20, WavePlayer_21, WavePlayer_22, WavePlayer_23]
  properties {weights [1,1,1,1,1,1]}
 }
 {id=WavePlayer_10, properties {soundWaveAssetPtr=/Game/Sounds/Fortress_Foley/Commando/Commando_Foley_Stereo_005}}
 {id=WavePlayer_11, properties {soundWaveAssetPtr=/Game/Sounds/Fortress_Foley/Commando/Commando_Foley_Stereo_009}}
 {id=WavePlayer_18, properties {soundWaveAssetPtr=/Game/Sounds/Fortress_Foley/Commando/Commando_Step_001}}
 {id=WavePlayer_19, properties {soundWaveAssetPtr=/Game/Sounds/Fortress_Foley/Commando/Commando_Step_002}}
 {id=WavePlayer_20, properties {soundWaveAssetPtr=/Game/Sounds/Fortress_Foley/Commando/Commando_Step_003}}
 {id=WavePlayer_21, properties {soundWaveAssetPtr=/Game/Sounds/Fortress_Foley/Commando/Commando_Step_004}}
 {id=WavePlayer_22, properties {soundWaveAssetPtr=/Game/Sounds/Fortress_Foley/Commando/Commando_Step_006}}
 {id=WavePlayer_23, properties {soundWaveAssetPtr=/Game/Sounds/Fortress_Foley/Commando/Commando_Step_010}}
 {id=WavePlayer_6, properties {soundWaveAssetPtr=/Game/Sounds/Fortress_Foley/Commando/Commando_Foley_Stereo_006}}
 {id=WavePlayer_7, properties {soundWaveAssetPtr=/Game/Sounds/Fortress_Foley/Commando/Commando_Foley_Stereo_008}}
 {id=WavePlayer_8, properties {soundWaveAssetPtr=/Game/Sounds/Fortress_Foley/Commando/Commando_Foley_Stereo_001}}
 {id=WavePlayer_9, properties {soundWaveAssetPtr=/Game/Sounds/Fortress_Foley/Commando/Commando_Foley_Stereo_003}}
]
properties {
 attenuationSettings=/Game/Sounds/Attenuation/Fort_Player_Footsteps_Att
 bOverrideConcurrency=true
 maxDistance=3500
 priority=0.5
 soundClassObject=&Fort_Player_Foley
 volumeMultiplier=0.4
 concurrencyOverrides {maxCount=4, resolutionRule=StopOldest}
}
