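# Sound cue for the Constructor's jump foley among Fortress player movement sounds.
# A Random node picks one of three equally weighted jump waves, then a Modulator varies pitch and volume downward.
# It plays in the player foley sound class with player footstep attenuation.
# A concurrency override allows one instance at a time and blocks new ones while it plays.
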SoundCue /Game/Sounds/Fortress_Foley/Player_Movement/Constructor_Jump_Cue
root=Modulator_2
nodes [
 {id=Modulator_2, children [Random_0], properties {pitchMax=1, pitchMin=0.8, volumeMax=1, volumeMin=0.7}}
 {id=Random_0, children [WavePlayer_275, WavePlayer_278, WavePlayer_280], properties {weights [1,1,1]}}
 {id=WavePlayer_275, properties {soundWaveAssetPtr=/Game/Sounds/Fortress_Foley/Constructor/Constructor_Jump_006}}
 {id=WavePlayer_278, properties {soundWaveAssetPtr=/Game/Sounds/Fortress_Foley/Constructor/Constructor_Jump_001}}
 {id=WavePlayer_280, properties {soundWaveAssetPtr=/Game/Sounds/Fortress_Foley/Constructor/Constructor_Jump_005}}
]
properties {
 attenuationSettings=/Game/Sounds/Attenuation/Fort_Player_Footsteps_Att
 bOverrideAttenuation=true
 bOverrideConcurrency=true
 maxDistance=4000
 soundClassObject=&Fort_Player_Foley
 volumeMultiplier=0.6
 concurrencyOverrides {maxCount=1, resolutionRule=PreventNew}
}
