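# Sound cue for the Constructor character's hard-landing foley, built from layered step and jump sounds.
# A mixer plays one of three random modulated step samples, plus one of two random jump samples, quieter and after a short randomized delay.
# Output uses the player foley sound class, footstep attenuation and a concurrency override that prevents new instances at the limit.
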SoundCue /Game/Sounds/Fortress_Foley/Player_Movement/Constructor_LandHard_Cue
root=Mixer_2
nodes [
 {id=Delay_2, children [Modulator_6], properties {delayMax=0.2, delayMin=0.1}}
 {id=Mixer_2, children [Modulator_5,Delay_2], properties {inputVolume [1,1]}}
 {id=Modulator_5, children [Random_30]}
 {id=Modulator_6, children [Random_31], properties {volumeMax=0.6, volumeMin=0.5}}
 {id=Random_30, children [WavePlayer_160, WavePlayer_161, WavePlayer_0], properties {weights [1,1,1]}}
 {id=Random_31, children [WavePlayer_163, WavePlayer_164], properties {weights [1,1]}}
 {id=WavePlayer_0, properties {soundWaveAssetPtr=/Game/Sounds/Fortress_Foley/Constructor/Constructor_Step_005}}
 {id=WavePlayer_160, properties {soundWaveAssetPtr=/Game/Sounds/Fort_Step_Sounds/Human/constructor_step_01}}
 {id=WavePlayer_161, properties {soundWaveAssetPtr=/Game/Sounds/Fort_Step_Sounds/Human/constructor_step_02}}
 {id=WavePlayer_163, properties {soundWaveAssetPtr=/Game/Sounds/Fort_Step_Sounds/Human/constructor_jump_01}}
 {id=WavePlayer_164, properties {soundWaveAssetPtr=/Game/Sounds/Fort_Step_Sounds/Human/constructor_jump_02}}
]
properties {
 attenuationSettings=/Game/Sounds/Attenuation/Fort_Player_Footsteps_Att
 bOverrideAttenuation=true
 bOverrideConcurrency=true
 maxDistance=4000
 soundClassObject=&Fort_Player_Foley
 volumeMultiplier=0.4
 concurrencyOverrides {resolutionRule=PreventNew}
}
