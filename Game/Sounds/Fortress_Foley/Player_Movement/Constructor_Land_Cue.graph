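# Sound cue for the Constructor character's landing foley, played in the player foley sound class with footstep attenuation.
# A mixer layers a louder random landing hit with a quieter random stereo foley layer that starts after a short random delay.
# Each layer picks one of three waves and varies its pitch and volume through its own modulator.
# A concurrency override prevents new instances once the limit is reached.
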SoundCue /Game/Sounds/Fortress_Foley/Player_Movement/Constructor_Land_Cue
root=Mixer_3
nodes [
 {id=Delay_0, children [Modulator_9], properties {delayMax=0.2}}
 {id=Mixer_3, children [Modulator_8,Delay_0], properties {inputVolume [1.7,0.2]}}
 {id=Modulator_8, children [Random_0], properties {pitchMax=1.2, pitchMin=1, volumeMax=1, volumeMin=0.7}}
 {id=Modulator_9, children [Random_1], properties {pitchMax=1.3, pitchMin=1, volumeMax=1, volumeMin=0.3}}
 {id=Random_0, children [WavePlayer_281, WavePlayer_282, WavePlayer_286], properties {weights [1,1,1]}}
 {id=Random_1, children [WavePlayer_287, WavePlayer_288, WavePlayer_290], properties {weights [1,1,1]}}
 {id=WavePlayer_281, properties {soundWaveAssetPtr=/Game/Sounds/Fortress_Foley/Constructor/Constructor_Land_006}}
 {id=WavePlayer_282, properties {soundWaveAssetPtr=/Game/Sounds/Fortress_Foley/Constructor/Constructor_Land_001}}
 {id=WavePlayer_286, properties {soundWaveAssetPtr=/Game/Sounds/Fortress_Foley/Constructor/Constructor_Land_005}}
 {id=WavePlayer_287, properties {soundWaveAssetPtr=/Game/Sounds/Fortress_Foley/Constructor/Constructor_Land_Foley_Stereo_001}}
 {id=WavePlayer_288, properties {soundWaveAssetPtr=/Game/Sounds/Fortress_Foley/Constructor/Constructor_Land_Foley_Stereo_002}}
 {id=WavePlayer_290, properties {soundWaveAssetPtr=/Game/Sounds/Fortress_Foley/Constructor/Constructor_Land_Foley_Stereo_004}}
]
properties {
 attenuationSettings=/Game/Sounds/Attenuation/Fort_Player_Footsteps_Att
 bOverrideConcurrency=true
 maxDistance=3500
 soundClassObject=&Fort_Player_Foley
 volumeMultiplier=0.3
 concurrencyOverrides {resolutionRule=PreventNew}
}
