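# Sound cue for the Constructor's running footsteps, routed to the player foley sound class with footstep attenuation.
# A mixer layers four randomized pools: Constructor steps at full volume, with quieter stereo foley, human "packsweet" walk and human player-walk layers.
# Each layer has its own pitch and volume modulation; the player-walk pool picks with replacement and includes an empty slot, so it sometimes drops out.
# A concurrency override limits simultaneous instances, stopping the oldest first.
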SoundCue /Game/Sounds/Fortress_Foley/Player_Movement/Constructor_Step_Run_Cue
root=Mixer_5
nodes [
 {id=Mixer_5, children [Modulator_17,Modulator_20,Modulator_23,Modulator_24], properties {inputVolume [1,0.15,0.25,0.4]}}
 {id=Modulator_17, children [Random_0], properties {pitchMax=1.3, pitchMin=1.1, volumeMax=1, volumeMin=0.7}}
 {id=Modulator_20, children [Random_1], properties {pitchMax=1.5, pitchMin=1.3, volumeMax=1, volumeMin=0.3}}
 {id=Modulator_23, children [Random_2], properties {pitchMax=1, pitchMin=0.8, volumeMax=1, volumeMin=0.3}}
 {id=Modulator_24, children [Random_20], properties {pitchMax=1.3, pitchMin=1.1, volumeMax=1, volumeMin=0.7}}
 {
  id=Random_0
  children [WavePlayer_0, WavePlayer_1, WavePlayer_14, WavePlayer_3, WavePlayer_4, WavePlayer_5]
  properties {weights [1,1,1,1,1,1]}
 }
 {
  id=Random_1
  children [WavePlayer_6, WavePlayer_7, WavePlayer_8, WavePlayer_9, WavePlayer_10, WavePlayer_11, WavePlayer_12]
  properties {weights [1,1,1,1,1,1,1]}
 }
 {
  id=Random_2
  children [WavePlayer_209, WavePlayer_210, WavePlayer_212, WavePlayer_213, WavePlayer_215, WavePlayer_216]
  properties {weights [1,1,1,1,1,1]}
 }
 {
  id=Random_20
  children [WavePlayer_221, WavePlayer_222, WavePlayer_223, WavePlayer_224, null]
  properties {bRandomizeWithoutReplacement=false, weights [1,1,1,1,1]}
 }
 {id=WavePlayer_0, properties {soundWaveAssetPtr=/Game/Sounds/Fortress_Foley/Constructor/Constructor_Step_001}}
 {id=WavePlayer_1, properties {soundWaveAssetPtr=/Game/Sounds/Fortress_Foley/Constructor/Constructor_Step_002}}
 {id=WavePlayer_10, properties {soundWaveAssetPtr=/Game/Sounds/Fortress_Foley/Constructor/Constructor_Foley_Stereo_001}}
 {id=WavePlayer_11, properties {soundWaveAssetPtr=/Game/Sounds/Fortress_Foley/Constructor/Constructor_Foley_Stereo_007}}
 {id=WavePlayer_12, properties {soundWaveAssetPtr=/Game/Sounds/Fortress_Foley/Constructor/Constructor_Foley_Stereo_009}}
 {id=WavePlayer_14, properties {soundWaveAssetPtr=/Game/Sounds/Fortress_Foley/Constructor/Constructor_Step_005}}
 {id=WavePlayer_209, properties {soundWaveAssetPtr=/Game/Sounds/Fort_Step_Sounds/Human/walk_packsweet_01}}
 {id=WavePlayer_210, properties {soundWaveAssetPtr=/Game/Sounds/Fort_Step_Sounds/Human/walk_packsweet_012}}
 {id=WavePlayer_212, properties {soundWaveAssetPtr=/Game/Sounds/Fort_Step_Sounds/Human/walk_packsweet_05}}
 {id=WavePlayer_213, properties {soundWaveAssetPtr=/Game/Sounds/Fort_Step_Sounds/Human/walk_packsweet_06}}
 {id=WavePlayer_215, properties {soundWaveAssetPtr=/Game/Sounds/Fort_Step_Sounds/Human/walk_packsweet_08}}
 {id=WavePlayer_216, properties {soundWaveAssetPtr=/Game/Sounds/Fort_Step_Sounds/Human/walk_packsweet_010}}
 {id=WavePlayer_221, properties {soundWaveAssetPtr=/Game/Sounds/Fort_Step_Sounds/Human/player_walk_04}}
 {id=WavePlayer_222, properties {soundWaveAssetPtr=/Game/Sounds/Fort_Step_Sounds/Human/player_walk_01}}
 {id=WavePlayer_223, properties {soundWaveAssetPtr=/Game/Sounds/Fort_Step_Sounds/Human/player_walk_02}}
 {id=WavePlayer_224, properties {soundWaveAssetPtr=/Game/Sounds/Fort_Step_Sounds/Human/player_walk_03}}
 {id=WavePlayer_3, properties {soundWaveAssetPtr=/Game/Sounds/Fortress_Foley/Constructor/Constructor_Step_007}}
 {id=WavePlayer_4, properties {soundWaveAssetPtr=/Game/Sounds/Fortress_Foley/Constructor/Constructor_Step_008}}
 {id=WavePlayer_5, properties {soundWaveAssetPtr=/Game/Sounds/Fortress_Foley/Constructor/Constructor_Step_012}}
 {id=WavePlayer_6, properties {soundWaveAssetPtr=/Game/Sounds/Fortress_Foley/Constructor/Constructor_Foley_Stereo_012}}
 {id=WavePlayer_7, properties {soundWaveAssetPtr=/Game/Sounds/Fortress_Foley/Constructor/Constructor_Foley_Stereo_002}}
 {id=WavePlayer_8, properties {soundWaveAssetPtr=/Game/Sounds/Fortress_Foley/Constructor/Constructor_Foley_Stereo_006}}
 {id=WavePlayer_9, properties {soundWaveAssetPtr=/Game/Sounds/Fortress_Foley/Constructor/Constructor_Foley_Stereo_010}}
]
properties {
 attenuationSettings=/Game/Sounds/Attenuation/Fort_Player_Footsteps_Att
 bOverrideConcurrency=true
 maxDistance=3500
 priority=0.5
 soundClassObject=&Fort_Player_Foley
 volumeMultiplier=0.2
 concurrencyOverrides {maxCount=4, resolutionRule=StopOldest}
}
