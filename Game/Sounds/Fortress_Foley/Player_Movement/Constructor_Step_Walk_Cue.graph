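# Sound cue for Constructor walking footsteps, routed through the player foley sound class with footstep attenuation and a low overall volume.
# A mixer layers two equally weighted random pools: step samples at full level and stereo foley samples at a much lower level.
# Each layer has its own pitch and volume modulator, with the foley pitched noticeably up.
# A concurrency override caps simultaneous instances, stopping the oldest.
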SoundCue /Game/Sounds/Fortress_Foley/Player_Movement/Constructor_Step_Walk_Cue
root=Mixer_4
nodes [
 {id=Mixer_4, children [Modulator_15,Modulator_16], properties {inputVolume [1,0.1]}}
 {id=Modulator_15, children [Random_3], properties {pitchMax=1.1, pitchMin=0.9, volumeMax=1, volumeMin=0.7}}
 {id=Modulator_16, children [Random_5], properties {pitchMax=1.8, pitchMin=1.4, volumeMax=1, volumeMin=0.3}}
 {
  id=Random_3
  children [WavePlayer_6, WavePlayer_7, WavePlayer_8, WavePlayer_9, WavePlayer_10, WavePlayer_11]
  properties {weights [1,1,1,1,1,1]}
 }
 {
  id=Random_5
  children [WavePlayer_19, WavePlayer_20, WavePlayer_21, WavePlayer_22, WavePlayer_23, WavePlayer_24, WavePlayer_25]
  properties {weights [1,1,1,1,1,1,1]}
 }
 {id=WavePlayer_10, properties {soundWaveAssetPtr=/Game/Sounds/Fortress_Foley/Constructor/Constructor_Step_008}}
 {id=WavePlayer_11, properties {soundWaveAssetPtr=/Game/Sounds/Fortress_Foley/Constructor/Constructor_Step_012}}
 {id=WavePlayer_19, properties {soundWaveAssetPtr=/Game/Sounds/Fortress_Foley/Constructor/Constructor_Foley_Stereo_012}}
 {id=WavePlayer_20, properties {soundWaveAssetPtr=/Game/Sounds/Fortress_Foley/Constructor/Constructor_Foley_Stereo_002}}
 {id=WavePlayer_21, properties {soundWaveAssetPtr=/Game/Sounds/Fortress_Foley/Constructor/Constructor_Foley_Stereo_006}}
 {id=WavePlayer_22, properties {soundWaveAssetPtr=/Game/Sounds/Fortress_Foley/Constructor/Constructor_Foley_Stereo_010}}
 {id=WavePlayer_23, properties {soundWaveAssetPtr=/Game/Sounds/Fortress_Foley/Constructor/Constructor_Foley_Stereo_001}}
 {id=WavePlayer_24, properties {soundWaveAssetPtr=/Game/Sounds/Fortress_Foley/Constructor/Constructor_Foley_Stereo_007}}
 {id=WavePlayer_25, properties {soundWaveAssetPtr=/Game/Sounds/Fortress_Foley/Constructor/Constructor_Foley_Stereo_009}}
 {id=WavePlayer_6, properties {soundWaveAssetPtr=/Game/Sounds/Fortress_Foley/Constructor/Constructor_Step_001}}
 {id=WavePlayer_7, properties {soundWaveAssetPtr=/Game/Sounds/Fortress_Foley/Constructor/Constructor_Step_002}}
 {id=WavePlayer_8, properties {soundWaveAssetPtr=/Game/Sounds/Fortress_Foley/Constructor/Constructor_Step_005}}
 {id=WavePlayer_9, properties {soundWaveAssetPtr=/Game/Sounds/Fortress_Foley/Constructor/Constructor_Step_007}}
]
properties {
 attenuationSettings=/Game/Sounds/Attenuation/Fort_Player_Footsteps_Att
 bOverrideConcurrency=true
 maxDistance=3500
 priority=0.5
 soundClassObject=&Fort_Player_Foley
 volumeMultiplier=0.15
 concurrencyOverrides {maxCount=4, resolutionRule=StopOldest}
}
