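# Sound cue for Ninja running footsteps in the player movement foley set.
# A Random node picks among five equally weighted ninja walk/run step waves, then feeds a Modulator for pitch and volume variation.
# It uses the player footsteps attenuation and player foley sound class at reduced volume.
# A concurrency override caps simultaneous steps, stopping the oldest first.
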SoundCue /Game/Sounds/Fortress_Foley/Player_Movement/Ninja_Step_Run_Cue
root=Modulator_7
nodes [
 {id=Modulator_7, children [Random_0]}
 {id=Random_0, children [WavePlayer_9, WavePlayer_10, WavePlayer_11, WavePlayer_12, WavePlayer_14], properties {weights [1,1,1,1,1]}}
 {id=WavePlayer_10, properties {soundWaveAssetPtr=/Game/Sounds/Fort_Step_Sounds/Human/ninja_walk_run_02}}
 {id=WavePlayer_11, properties {soundWaveAssetPtr=/Game/Sounds/Fort_Step_Sounds/Human/ninja_walk_run_03}}
 {id=WavePlayer_12, properties {soundWaveAssetPtr=/Game/Sounds/Fort_Step_Sounds/Human/ninja_walk_run_04}}
 {id=WavePlayer_14, properties {soundWaveAssetPtr=/Game/Sounds/Fort_Step_Sounds/Human/ninja_walk_run_06}}
 {id=WavePlayer_9, properties {soundWaveAssetPtr=/Game/Sounds/Fort_Step_Sounds/Human/ninja_walk_run_01}}
]
properties {
 attenuationSettings=/Game/Sounds/Attenuation/Fort_Player_Footsteps_Att
 bOverrideConcurrency=true
 maxDistance=3500
 priority=0.5
 soundClassObject=&Fort_Player_Foley
 volumeMultiplier=0.5
 concurrencyOverrides {maxCount=4, resolutionRule=StopOldest}
}
