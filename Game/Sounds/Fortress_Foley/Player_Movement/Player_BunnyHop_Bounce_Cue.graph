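# Sound cue for the player bunny-hop bounce foley, built from human soldier running footstep samples.
# A Random node picks one of five equally weighted step waves and feeds a Modulator that varies volume.
# The cue plays quiet and slightly pitched up, routed to the player foley sound class with an emote attenuation asset.
# A concurrency override caps simultaneous instances, stopping the oldest first.
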SoundCue /Game/Sounds/Fortress_Foley/Player_Movement/Player_BunnyHop_Bounce_Cue
root=Modulator_0
nodes [
 {id=Modulator_0, children [Random_7], properties {volumeMin=0.75}}
 {id=Random_7, children [WavePlayer_60, WavePlayer_82, WavePlayer_83, WavePlayer_84, WavePlayer_102], properties {weights [1,1,1,1,1]}}
 {id=WavePlayer_102, properties {soundWaveAssetPtr=/Game/Sounds/Fort_Step_Sounds/Human/soldier_run_step_05}}
 {id=WavePlayer_60, properties {soundWaveAssetPtr=/Game/Sounds/Fort_Step_Sounds/Human/soldier_run_step_01}}
 {id=WavePlayer_82, properties {soundWaveAssetPtr=/Game/Sounds/Fort_Step_Sounds/Human/soldier_run_step_02}}
 {id=WavePlayer_83, properties {soundWaveAssetPtr=/Game/Sounds/Fort_Step_Sounds/Human/soldier_run_step_03}}
 {id=WavePlayer_84, properties {soundWaveAssetPtr=/Game/Sounds/Fort_Step_Sounds/Human/soldier_run_step_04}}
]
properties {
 attenuationSettings=&Fort_Emotes_Athena_Att
 bOverrideConcurrency=true
 maxDistance=1600
 pitchMultiplier=1.2
 priority=0.5
 soundClassObject=&Fort_Player_Foley
 volumeMultiplier=0.15
 concurrencyOverrides {maxCount=4, resolutionRule=StopOldest}
}
