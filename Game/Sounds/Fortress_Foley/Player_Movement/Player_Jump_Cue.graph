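# Sound cue for the player's jump foley, routed to the player foley sound class.
# A random node picks among three equally weighted Outlander jump waves and feeds a modulator that varies pitch and volume on each play.
# It uses player footstep attenuation and overrides concurrency to allow a single instance, blocking new plays while one is active.
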SoundCue /Game/Sounds/Fortress_Foley/Player_Movement/Player_Jump_Cue
root=Modulator_6
nodes [
 {id=Modulator_6, children [Random_1], properties {pitchMax=1.1, pitchMin=0.9, volumeMax=1, volumeMin=0.7}}
 {id=Random_1, children [WavePlayer_0,WavePlayer_2,WavePlayer_4], properties {weights [1,1,1]}}
 {id=WavePlayer_0, properties {soundWaveAssetPtr=/Game/Sounds/Fortress_Foley/Outlander/Outlander_Jump_003}}
 {id=WavePlayer_2, properties {soundWaveAssetPtr=/Game/Sounds/Fortress_Foley/Outlander/Outlander_Jump_004}}
 {id=WavePlayer_4, properties {soundWaveAssetPtr=/Game/Sounds/Fortress_Foley/Outlander/Outlander_Jump_002}}
]
properties {
 attenuationSettings=/Game/Sounds/Attenuation/Fort_Player_Footsteps_Att
 bOverrideConcurrency=true
 maxDistance=3500
 pitchMultiplier=1.1
 soundClassObject=&Fort_Player_Foley
 volumeMultiplier=0.8
 concurrencyOverrides {maxCount=1, resolutionRule=PreventNew}
}
