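# Player landing Sound Cue that plays foley through footstep attenuation, with a concurrency override that prevents new instances at the limit.
# A root Switch selects among layered mixes, each pairing randomized Outlander land and stereo foley samples with pitch and volume modulation.
# One branch adds wood steps, one delayed, plus an occasional delayed squeak via a weighted empty Random slot.
# Two others add hollow metal or metal stair landings.
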SoundCue /Game/Sounds/Fortress_Foley/Player_Movement/Player_Land_Cue
root=Switch_0
nodes [
 {id=Delay_2, children [WavePlayer_60], properties {delayMin=0.2}}
 {id=Delay_3, children [Random_19], properties {delayMin=0.2}}
 {id=Mixer_1, children [Modulator_2,Modulator_3], properties {inputVolume [1,0.6]}}
 {id=Mixer_4, children [Modulator_8,Modulator_9,Delay_3,Mixer_9], properties {inputVolume [1,0.6,1,1]}}
 {id=Mixer_5, children [Modulator_10,Modulator_11], properties {inputVolume [1,0.6]}}
 {id=Mixer_6, children [Modulator_1,Modulator_13,Modulator_4], properties {inputVolume [1,0.6,1]}}
 {id=Mixer_7, children [Modulator_6,Modulator_15,Modulator_7], properties {inputVolume [1,0.6,1]}}
 {id=Mixer_8, children [Modulator_16,Modulator_17], properties {inputVolume [1,0.6]}}
 {id=Mixer_9, children [Modulator_20,Modulator_21], properties {inputVolume [1,1]}}
 {id=Modulator_1, children [Random_12], properties {pitchMax=1.5, pitchMin=1.3, volumeMax=1.2, volumeMin=0.7}}
 {id=Modulator_10, children [Random_10], properties {pitchMax=1.5, pitchMin=1.3, volumeMax=1.3, volumeMin=0.7}}
 {id=Modulator_11, children [Random_11], properties {pitchMax=1.2, pitchMin=1, volumeMax=1.3, volumeMin=0.7}}
 {id=Modulator_13, children [Random_13], properties {pitchMax=1.2, pitchMin=1, volumeMax=0.65, volumeMin=0.3}}
 {id=Modulator_15, children [Random_15], properties {pitchMax=1.2, pitchMin=1, volumeMax=0.65, volumeMin=0.3}}
 {id=Modulator_16, children [Random_16], properties {pitchMax=1.5, pitchMin=1.3, volumeMax=1.3, volumeMin=0.7}}
 {id=Modulator_17, children [Random_17], properties {pitchMax=1.2, pitchMin=1, volumeMax=1.3, volumeMin=0.7}}
 {id=Modulator_2, children [Random_2], properties {pitchMax=1.5, pitchMin=1.3, volumeMax=1.3, volumeMin=0.7}}
 {id=Modulator_20, children [Delay_2], properties {pitchMax=0.95, pitchMin=0.94, volumeMax=1, volumeMin=1}}
 {id=Modulator_21, children [Random_20], properties {pitchMax=0.95, pitchMin=0.94, volumeMax=1, volumeMin=1}}
 {id=Modulator_3, children [Random_3], properties {pitchMax=1.2, pitchMin=1, volumeMax=1.3, volumeMin=0.7}}
 {id=Modulator_4, children [Random_0], properties {pitchMax=1.7, pitchMin=1.3, volumeMax=1.3, volumeMin=0.7}}
 {id=Modulator_6, children [Random_14], properties {pitchMax=1.5, pitchMin=1.3, volumeMax=1.3, volumeMin=0.7}}
 {id=Modulator_7, children [Random_1], properties {pitchMax=1.2, pitchMin=0.94, volumeMax=1.3, volumeMin=0.7}}
 {id=Modulator_8, children [Random_8], properties {pitchMax=1.5, pitchMin=1.3, volumeMax=1.3, volumeMin=0.7}}
 {id=Modulator_9, children [Random_9], properties {pitchMax=1.2, pitchMin=1, volumeMax=1.3, volumeMin=0.7}}
 {id=Random_0, children [WavePlayer_0,WavePlayer_1,WavePlayer_2,WavePlayer_3], properties {weights [1,1,1,1]}}
 {id=Random_1, children [WavePlayer_4, WavePlayer_5, WavePlayer_12, WavePlayer_13], properties {weights [1,1,1,1]}}
 {id=Random_10, children [WavePlayer_30, WavePlayer_31, WavePlayer_32], properties {weights [1,1,1]}}
 {id=Random_11, children [WavePlayer_33, WavePlayer_34, WavePlayer_35], properties {weights [1,1,1]}}
 {id=Random_12, children [WavePlayer_36, WavePlayer_37, WavePlayer_38], properties {weights [1,1,1]}}
 {id=Random_13, children [WavePlayer_39, WavePlayer_40, WavePlayer_41], properties {weights [1,1,1]}}
 {id=Random_14, children [WavePlayer_42, WavePlayer_43, WavePlayer_44], properties {weights [1,1,1]}}
 {id=Random_15, children [WavePlayer_45, WavePlayer_46, WavePlayer_47], properties {weights [1,1,1]}}
 {id=Random_16, children [WavePlayer_48, WavePlayer_49, WavePlayer_50], properties {weights [1,1,1]}}
 {id=Random_17, children [WavePlayer_51, WavePlayer_52, WavePlayer_53], properties {weights [1,1,1]}}
 {
  id=Random_18
  children [WavePlayer_54, WavePlayer_55, WavePlayer_56, WavePlayer_57, WavePlayer_58, WavePlayer_59]
  properties {weights [1,1,1,1,1,1]}
 }
 {id=Random_19, children [Random_18,null], properties {weights [0.7,1]}}
 {id=Random_2, children [WavePlayer_6,WavePlayer_7,WavePlayer_8], properties {weights [1,1,1]}}
 {id=Random_20, children [WavePlayer_61, WavePlayer_62], properties {weights [1,1]}}
 {id=Random_3, children [WavePlayer_9, WavePlayer_10, WavePlayer_11], properties {weights [1,1,1]}}
 {id=Random_8, children [WavePlayer_24, WavePlayer_25, WavePlayer_26], properties {weights [1,1,1]}}
 {id=Random_9, children [WavePlayer_27, WavePlayer_28, WavePlayer_29], properties {weights [1,1,1]}}
 {id=Switch_0, children [Mixer_1,Mixer_1,Mixer_4,Mixer_5,Mixer_6,Mixer_7,Mixer_8]}
 {id=WavePlayer_0, properties {soundWaveAssetPtr=/Game/Sounds/Fort_Step_Sounds/Human/FS_Metal_Hollow_MedHvy_Land_000}}
 {id=WavePlayer_1, properties {soundWaveAssetPtr=/Game/Sounds/Fort_Step_Sounds/Human/FS_Metal_Hollow_MedHvy_Land_001}}
 {id=WavePlayer_10, properties {soundWaveAssetPtr=/Game/Sounds/Fortress_Foley/Outlander/Outlander_Land_Foley_Stereo_003}}
 {id=WavePlayer_11, properties {soundWaveAssetPtr=/Game/Sounds/Fortress_Foley/Outlander/Outlander_Land_Foley_Stereo_004}}
 {id=WavePlayer_12, properties {soundWaveAssetPtr=/Game/Sounds/Fort_Step_Sounds/Human/FS_Metal_Stair_Med_Land_002}}
 {id=WavePlayer_13, properties {soundWaveAssetPtr=/Game/Sounds/Fort_Step_Sounds/Human/FS_Metal_Stair_Med_Land_003}}
 {id=WavePlayer_2, properties {soundWaveAssetPtr=/Game/Sounds/Fort_Step_Sounds/Human/FS_Metal_Hollow_MedHvy_Land_002}}
 {id=WavePlayer_24, properties {soundWaveAssetPtr=/Game/Sounds/Fortress_Foley/Outlander/Outlander_Land_003}}
 {id=WavePlayer_25, properties {soundWaveAssetPtr=/Game/Sounds/Fortress_Foley/Outlander/Outlander_Land_005}}
 {id=WavePlayer_26, properties {soundWaveAssetPtr=/Game/Sounds/Fortress_Foley/Outlander/Outlander_Land_004}}
 {id=WavePlayer_27, properties {soundWaveAssetPtr=/Game/Sounds/Fortress_Foley/Outlander/Outlander_Land_Foley_Stereo_001}}
 {id=WavePlayer_28, properties {soundWaveAssetPtr=&Outlander_Land_Foley_Stereo_003}}
 {id=WavePlayer_29, properties {soundWaveAssetPtr=&Outlander_Land_Foley_Stereo_004}}
 {id=WavePlayer_3, properties {soundWaveAssetPtr=/Game/Sounds/Fort_Step_Sounds/Human/FS_Metal_Hollow_MedHvy_Land_003}}
 {id=WavePlayer_30, properties {soundWaveAssetPtr=&Outlander_Land_003}}
 {id=WavePlayer_31, properties {soundWaveAssetPtr=&Outlander_Land_005}}
 {id=WavePlayer_32, properties {soundWaveAssetPtr=&Outlander_Land_004}}
 {id=WavePlayer_33, properties {soundWaveAssetPtr=&Outlander_Land_Foley_Stereo_001}}
 {id=WavePlayer_34, properties {soundWaveAssetPtr=&Outlander_Land_Foley_Stereo_003}}
 {id=WavePlayer_35, properties {soundWaveAssetPtr=&Outlander_Land_Foley_Stereo_004}}
 {id=WavePlayer_36, properties {soundWaveAssetPtr=&Outlander_Land_003}}
 {id=WavePlayer_37, properties {soundWaveAssetPtr=&Outlander_Land_005}}
 {id=WavePlayer_38, properties {soundWaveAssetPtr=&Outlander_Land_004}}
 {id=WavePlayer_39, properties {soundWaveAssetPtr=&Outlander_Land_Foley_Stereo_001}}
 {id=WavePlayer_4, properties {soundWaveAssetPtr=/Game/Sounds/Fort_Step_Sounds/Human/FS_Metal_Stair_Med_Land_000}}
 {id=WavePlayer_40, properties {soundWaveAssetPtr=&Outlander_Land_Foley_Stereo_003}}
 {id=WavePlayer_41, properties {soundWaveAssetPtr=&Outlander_Land_Foley_Stereo_004}}
 {id=WavePlayer_42, properties {soundWaveAssetPtr=&Outlander_Land_003}}
 {id=WavePlayer_43, properties {soundWaveAssetPtr=&Outlander_Land_005}}
 {id=WavePlayer_44, properties {soundWaveAssetPtr=&Outlander_Land_004}}
 {id=WavePlayer_45, properties {soundWaveAssetPtr=&Outlander_Land_Foley_Stereo_001}}
 {id=WavePlayer_46, properties {soundWaveAssetPtr=&Outlander_Land_Foley_Stereo_003}}
 {id=WavePlayer_47, properties {soundWaveAssetPtr=&Outlander_Land_Foley_Stereo_004}}
 {id=WavePlayer_48, properties {soundWaveAssetPtr=&Outlander_Land_003}}
 {id=WavePlayer_49, properties {soundWaveAssetPtr=&Outlander_Land_005}}
 {id=WavePlayer_5, properties {soundWaveAssetPtr=/Game/Sounds/Fort_Step_Sounds/Human/FS_Metal_Stair_Med_Land_001}}
 {id=WavePlayer_50, properties {soundWaveAssetPtr=&Outlander_Land_004}}
 {id=WavePlayer_51, properties {soundWaveAssetPtr=&Outlander_Land_Foley_Stereo_001}}
 {id=WavePlayer_52, properties {soundWaveAssetPtr=&Outlander_Land_Foley_Stereo_003}}
 {id=WavePlayer_53, properties {soundWaveAssetPtr=&Outlander_Land_Foley_Stereo_004}}
 {id=WavePlayer_54, properties {soundWaveAssetPtr=/Game/Sounds/Fort_Step_Sounds/Human/FS_Wood_Generic_Squeaks_000}}
 {id=WavePlayer_55, properties {soundWaveAssetPtr=/Game/Sounds/Fort_Step_Sounds/Human/FS_Wood_Generic_Squeaks_001}}
 {id=WavePlayer_56, properties {soundWaveAssetPtr=/Game/Sounds/Fort_Step_Sounds/Human/FS_Wood_Generic_Squeaks_002}}
 {id=WavePlayer_57, properties {soundWaveAssetPtr=/Game/Sounds/Fort_Step_Sounds/Human/FS_Wood_Generic_Squeaks_003}}
 {id=WavePlayer_58, properties {soundWaveAssetPtr=/Game/Sounds/Fort_Step_Sounds/Human/FS_Wood_Generic_Squeaks_004}}
 {id=WavePlayer_59, properties {soundWaveAssetPtr=/Game/Sounds/Fort_Step_Sounds/Human/FS_Wood_Generic_Squeaks_005}}
 {id=WavePlayer_6, properties {soundWaveAssetPtr=&Outlander_Land_003}}
 {id=WavePlayer_60, properties {soundWaveAssetPtr=/Game/Sounds/Fort_Step_Sounds/Human/FS_Wood_Walk_MedHvy_Generic_004}}
 {id=WavePlayer_61, properties {soundWaveAssetPtr=/Game/Sounds/Fort_Step_Sounds/Human/FS_Wood_Walk_MedHvy_Generic_002}}
 {id=WavePlayer_62, properties {soundWaveAssetPtr=/Game/Sounds/Fort_Step_Sounds/Human/FS_Wood_Walk_MedHvy_Generic_016}}
 {id=WavePlayer_7, properties {soundWaveAssetPtr=&Outlander_Land_005}}
 {id=WavePlayer_8, properties {soundWaveAssetPtr=&Outlander_Land_004}}
 {id=WavePlayer_9, properties {soundWaveAssetPtr=&Outlander_Land_Foley_Stereo_001}}
]
properties {
 attenuationSettings=/Game/Sounds/Attenuation/Fort_Player_Footsteps_Att
 bOverrideConcurrency=true
 maxDistance=3500
 soundClassObject=&Fort_Player_Foley
 volumeMultiplier=1
 concurrencyOverrides {resolutionRule=PreventNew}
}
