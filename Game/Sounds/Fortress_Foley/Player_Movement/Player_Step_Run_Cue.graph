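# Sound cue for player running footsteps, with footstep attenuation and a four-voice limit that stops the oldest.
# A root switch picks among seven slots, apparently by surface: wood, two metal variants, and a repeated generic Outlander step/dirt mix.
# Each mixer layers randomized steps and quieter Outlander foley through pitch and volume modulators.
# Wood squeaks and hollow-metal hits play only occasionally, via weighted randoms with an empty branch.
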SoundCue /Game/Sounds/Fortress_Foley/Player_Movement/Player_Step_Run_Cue
root=Switch_1
nodes [
 {id=Mixer_1, children [Modulator_4,Modulator_3,Modulator_5], properties {inputVolume [0.7,0.2,1]}}
 {id=Mixer_11, children [Modulator_34,Modulator_33,Modulator_35], properties {inputVolume [0.7,0.2,1]}}
 {id=Mixer_12, children [Modulator_37,Modulator_36,Modulator_38], properties {inputVolume [0.7,0.2,1]}}
 {id=Mixer_3, children [Modulator_10,Modulator_9,Modulator_11,Random_11], properties {inputVolume [0.7,0.2,1,1]}}
 {id=Mixer_5, children [Modulator_16,Modulator_15,Modulator_17], properties {inputVolume [0.7,0.2,1]}}
 {id=Mixer_8, children [Modulator_1,Modulator_24,Modulator_26,Modulator_2], properties {inputVolume [0.7,0.2,1,1]}}
 {id=Modulator_1, children [Random_6], properties {pitchMax=0.96, pitchMin=0.94, volumeMax=1, volumeMin=0.7}}
 {id=Modulator_10, children [Random_0], properties {pitchMax=1.8, pitchMin=1.4, volumeMax=1, volumeMin=0.7}}
 {id=Modulator_11, children [Random_1], properties {pitchMax=1.4, pitchMin=1.1, volumeMax=1, volumeMin=0.5}}
 {id=Modulator_15, children [Random_16], properties {pitchMax=1.6, pitchMin=1.4, volumeMax=1, volumeMin=0.3}}
 {id=Modulator_16, children [Random_15], properties {pitchMax=1.8, pitchMin=1.4, volumeMax=1, volumeMin=0.7}}
 {id=Modulator_17, children [Random_17], properties {pitchMax=1.4, pitchMin=1.1, volumeMax=1, volumeMin=0.5}}
 {id=Modulator_2, children [Random_12], properties {pitchMax=1.4, pitchMin=1.2, volumeMax=0.4, volumeMin=0.3}}
 {id=Modulator_24, children [Random_25], properties {pitchMax=1.6, pitchMin=1.4, volumeMax=1, volumeMin=0.3}}
 {id=Modulator_26, children [Random_13], properties {pitchMax=1.4, pitchMin=1.1, volumeMax=1, volumeMin=0.5}}
 {id=Modulator_3, children [Random_4], properties {pitchMax=1.6, pitchMin=1.4, volumeMax=1, volumeMin=0.3}}
 {id=Modulator_33, children [Random_34], properties {pitchMax=1.6, pitchMin=1.4, volumeMax=1, volumeMin=0.3}}
 {id=Modulator_34, children [Random_2], properties {pitchMax=1, pitchMin=0.94, volumeMax=1, volumeMin=0.7}}
 {id=Modulator_35, children [Random_35], properties {pitchMax=1.4, pitchMin=1.1, volumeMax=1, volumeMin=0.5}}
 {id=Modulator_36, children [Random_37], properties {pitchMax=1.6, pitchMin=1.4, volumeMax=1, volumeMin=0.3}}
 {id=Modulator_37, children [Random_36], properties {pitchMax=1.8, pitchMin=1.4, volumeMax=1, volumeMin=0.7}}
 {id=Modulator_38, children [Random_38], properties {pitchMax=1.4, pitchMin=1.1, volumeMax=1, volumeMin=0.5}}
 {id=Modulator_4, children [Random_3], properties {pitchMax=1.8, pitchMin=1.4, volumeMax=1, volumeMin=0.7}}
 {id=Modulator_5, children [Random_5], properties {pitchMax=1.4, pitchMin=1.1, volumeMax=1, volumeMin=0.5}}
 {id=Modulator_9, children [Random_10], properties {pitchMax=1.6, pitchMin=1.4, volumeMax=1, volumeMin=0.3}}
 {
  id=Random_0
  children [WavePlayer_0,WavePlayer_1,WavePlayer_2,WavePlayer_3,WavePlayer_4,WavePlayer_5,WavePlayer_6,WavePlayer_7,WavePlayer_8]
  properties {weights [1,1,1,1,1,1,1,1,1]}
 }
 {
  id=Random_1
  children [WavePlayer_9, WavePlayer_10, WavePlayer_11, WavePlayer_12, WavePlayer_13, WavePlayer_14, WavePlayer_15, WavePlayer_16]
  properties {weights [1,1,1,1,1,1,1,1]}
 }
 {
  id=Random_10
  children [WavePlayer_62, WavePlayer_63, WavePlayer_64, WavePlayer_65, WavePlayer_66, WavePlayer_67]
  properties {weights [1,1,1,1,1,1]}
 }
 {id=Random_11, children [Random_9,null], properties {bRandomizeWithoutReplacement=false, weights [0.5,1]}}
 {id=Random_12, children [Random_8,null], properties {weights [0.7,1]}}
 {
  id=Random_13
  children [WavePlayer_74, WavePlayer_75, WavePlayer_76, WavePlayer_77, WavePlayer_78, WavePlayer_79, WavePlayer_80, WavePlayer_81]
  properties {weights [1,1,1,1,1,1,1,1]}
 }
 {
  id=Random_15
  children [WavePlayer_90, WavePlayer_91, WavePlayer_92, WavePlayer_93, WavePlayer_94, WavePlayer_95]
  properties {weights [1,1,1,1,1,1]}
 }
 {
  id=Random_16
  children [WavePlayer_96, WavePlayer_97, WavePlayer_98, WavePlayer_99, WavePlayer_100, WavePlayer_101]
  properties {weights [1,1,1,1,1,1]}
 }
 {id=Random_17, children [WavePlayer_85, WavePlayer_86, WavePlayer_87, WavePlayer_88, WavePlayer_89], properties {weights [1,1,1,1,1]}}
 {
  id=Random_2
  children [WavePlayer_43, WavePlayer_44, WavePlayer_45, WavePlayer_46, WavePlayer_47, WavePlayer_48, WavePlayer_49, WavePlayer_50]
  properties {weights [1,1,1,1,1,1,1,1]}
 }
 {
  id=Random_25
  children [WavePlayer_147, WavePlayer_148, WavePlayer_149, WavePlayer_150, WavePlayer_151, WavePlayer_152]
  properties {weights [1,1,1,1,1,1]}
 }
 {
  id=Random_3
  children [WavePlayer_22, WavePlayer_23, WavePlayer_24, WavePlayer_25, WavePlayer_26, WavePlayer_27]
  properties {weights [1,1,1,1,1,1]}
 }
 {
  id=Random_34
  children [WavePlayer_198, WavePlayer_199, WavePlayer_200, WavePlayer_201, WavePlayer_202, WavePlayer_203]
  properties {weights [1,1,1,1,1,1]}
 }
 {id=Random_35, children [WavePlayer_187, WavePlayer_188, WavePlayer_189, WavePlayer_190, WavePlayer_191], properties {weights [1,1,1,1,1]}}
 {
  id=Random_36
  children [WavePlayer_209, WavePlayer_210, WavePlayer_211, WavePlayer_212, WavePlayer_213, WavePlayer_214]
  properties {weights [1,1,1,1,1,1]}
 }
 {
  id=Random_37
  children [WavePlayer_215, WavePlayer_216, WavePlayer_217, WavePlayer_218, WavePlayer_219, WavePlayer_220]
  properties {weights [1,1,1,1,1,1]}
 }
 {id=Random_38, children [WavePlayer_204, WavePlayer_205, WavePlayer_206, WavePlayer_207, WavePlayer_208], properties {weights [1,1,1,1,1]}}
 {
  id=Random_4
  children [WavePlayer_28, WavePlayer_29, WavePlayer_30, WavePlayer_31, WavePlayer_32, WavePlayer_33]
  properties {weights [1,1,1,1,1,1]}
 }
 {id=Random_5, children [WavePlayer_17, WavePlayer_18, WavePlayer_19, WavePlayer_20, WavePlayer_21], properties {weights [1,1,1,1,1]}}
 {
  id=Random_6
  children [WavePlayer_51, WavePlayer_52, WavePlayer_53, WavePlayer_54, WavePlayer_55, WavePlayer_56, WavePlayer_57, WavePlayer_58, WavePlayer_59]
  properties {weights [1,1,1,1,1,1,1,1,1]}
 }
 {
  id=Random_8
  children [WavePlayer_61, WavePlayer_68, WavePlayer_69, WavePlayer_70, WavePlayer_71, WavePlayer_72, WavePlayer_73]
  properties {weights [1,1,1,1,1,1,1]}
 }
 {
  id=Random_9
  children [WavePlayer_34, WavePlayer_35, WavePlayer_36, WavePlayer_37, WavePlayer_38, WavePlayer_39, WavePlayer_40, WavePlayer_41, WavePlayer_42]
  properties {weights [1,1,1,1,1,1,1,1,1]}
 }
 {id=Switch_1, children [Mixer_1,Mixer_1,Mixer_3,Mixer_5,Mixer_8,Mixer_11,Mixer_12]}
 {id=WavePlayer_0, properties {soundWaveAssetPtr=/Game/Sounds/Fort_Step_Sounds/Human/FS_Wood_Generic_Med_008}}
 {id=WavePlayer_1, properties {soundWaveAssetPtr=/Game/Sounds/Fort_Step_Sounds/Human/FS_Wood_Generic_Med_000}}
 {id=WavePlayer_10, properties {soundWaveAssetPtr=/Game/Sounds/Fort_Step_Sounds/Human/FS_Wood_Generic_MedHvy_001}}
 {id=WavePlayer_100, properties {soundWaveAssetPtr=/Game/Sounds/Fortress_Foley/Outlander/Outlander_Foley_Stereo_003}}
 {id=WavePlayer_101, properties {soundWaveAssetPtr=/Game/Sounds/Fortress_Foley/Outlander/Outlander_Foley_Stereo_005}}
 {id=WavePlayer_11, properties {soundWaveAssetPtr=/Game/Sounds/Fort_Step_Sounds/Human/FS_Wood_Generic_MedHvy_002}}
 {id=WavePlayer_12, properties {soundWaveAssetPtr=/Game/Sounds/Fort_Step_Sounds/Human/FS_Wood_Generic_MedHvy_003}}
 {id=WavePlayer_13, properties {soundWaveAssetPtr=/Game/Sounds/Fort_Step_Sounds/Human/FS_Wood_Generic_MedHvy_004}}
 {id=WavePlayer_14, properties {soundWaveAssetPtr=/Game/Sounds/Fort_Step_Sounds/Human/FS_Wood_Generic_MedHvy_005}}
 {id=WavePlayer_147, properties {soundWaveAssetPtr=/Game/Sounds/Fortress_Foley/Outlander/Outlander_Foley_Stereo_002}}
 {id=WavePlayer_148, properties {soundWaveAssetPtr=/Game/Sounds/Fortress_Foley/Outlander/Outlander_Foley_Stereo_006}}
 {id=WavePlayer_149, properties {soundWaveAssetPtr=/Game/Sounds/Fortress_Foley/Outlander/Outlander_Foley_Stereo_010}}
 {id=WavePlayer_15, properties {soundWaveAssetPtr=/Game/Sounds/Fort_Step_Sounds/Human/FS_Wood_Generic_MedHvy_006}}
 {id=WavePlayer_150, properties {soundWaveAssetPtr=/Game/Sounds/Fortress_Foley/Outlander/Outlander_Foley_Stereo_001}}
 {id=WavePlayer_151, properties {soundWaveAssetPtr=&Outlander_Foley_Stereo_003}}
 {id=WavePlayer_152, properties {soundWaveAssetPtr=&Outlander_Foley_Stereo_005}}
 {id=WavePlayer_16, properties {soundWaveAssetPtr=/Game/Sounds/Fort_Step_Sounds/Human/FS_Wood_Generic_MedHvy_007}}
 {id=WavePlayer_17, properties {soundWaveAssetPtr=/Game/Sounds/Fort_Step_Sounds/Human/footsteps_dirt_trail_07}}
 {id=WavePlayer_18, properties {soundWaveAssetPtr=/Game/Sounds/Fort_Step_Sounds/Human/footsteps_dirt_trail_05}}
 {id=WavePlayer_187, properties {soundWaveAssetPtr=&footsteps_dirt_trail_07}}
 {id=WavePlayer_188, properties {soundWaveAssetPtr=&footsteps_dirt_trail_05}}
 {id=WavePlayer_189, properties {soundWaveAssetPtr=/Game/Sounds/Fort_Step_Sounds/Human/footsteps_dirt_trail_03}}
 {id=WavePlayer_19, properties {soundWaveAssetPtr=&footsteps_dirt_trail_03}}
 {id=WavePlayer_190, properties {soundWaveAssetPtr=/Game/Sounds/Fort_Step_Sounds/Human/footsteps_dirt_trail_02}}
 {id=WavePlayer_191, properties {soundWaveAssetPtr=/Game/Sounds/Fort_Step_Sounds/Human/footsteps_dirt_trail_01}}
 {id=WavePlayer_198, properties {soundWaveAssetPtr=&Outlander_Foley_Stereo_002}}
 {id=WavePlayer_199, properties {soundWaveAssetPtr=&Outlander_Foley_Stereo_006}}
 {id=WavePlayer_2, properties {soundWaveAssetPtr=/Game/Sounds/Fort_Step_Sounds/Human/FS_Wood_Generic_Med_001}}
 {id=WavePlayer_20, properties {soundWaveAssetPtr=&footsteps_dirt_trail_02}}
 {id=WavePlayer_200, properties {soundWaveAssetPtr=&Outlander_Foley_Stereo_010}}
 {id=WavePlayer_201, properties {soundWaveAssetPtr=&Outlander_Foley_Stereo_001}}
 {id=WavePlayer_202, properties {soundWaveAssetPtr=&Outlander_Foley_Stereo_003}}
 {id=WavePlayer_203, properties {soundWaveAssetPtr=&Outlander_Foley_Stereo_005}}
 {id=WavePlayer_204, properties {soundWaveAssetPtr=&footsteps_dirt_trail_07}}
 {id=WavePlayer_205, properties {soundWaveAssetPtr=&footsteps_dirt_trail_05}}
 {id=WavePlayer_206, properties {soundWaveAssetPtr=&footsteps_dirt_trail_03}}
 {id=WavePlayer_207, properties {soundWaveAssetPtr=&footsteps_dirt_trail_02}}
 {id=WavePlayer_208, properties {soundWaveAssetPtr=&footsteps_dirt_trail_01}}
 {id=WavePlayer_209, properties {soundWaveAssetPtr=/Game/Sounds/Fortress_Foley/Outlander/Outlander_Step_001}}
 {id=WavePlayer_21, properties {soundWaveAssetPtr=&footsteps_dirt_trail_01}}
 {id=WavePlayer_210, properties {soundWaveAssetPtr=/Game/Sounds/Fortress_Foley/Outlander/Outlander_Step_003}}
 {id=WavePlayer_211, properties {soundWaveAssetPtr=/Game/Sounds/Fortress_Foley/Outlander/Outlander_Step_004}}
 {id=WavePlayer_212, properties {soundWaveAssetPtr=/Game/Sounds/Fortress_Foley/Outlander/Outlander_Step_005}}
 {id=WavePlayer_213, properties {soundWaveAssetPtr=/Game/Sounds/Fortress_Foley/Outlander/Outlander_Step_007}}
 {id=WavePlayer_214, properties {soundWaveAssetPtr=/Game/Sounds/Fortress_Foley/Outlander/Outlander_Step_012}}
 {id=WavePlayer_215, properties {soundWaveAssetPtr=&Outlander_Foley_Stereo_002}}
 {id=WavePlayer_216, properties {soundWaveAssetPtr=&Outlander_Foley_Stereo_006}}
 {id=WavePlayer_217, properties {soundWaveAssetPtr=&Outlander_Foley_Stereo_010}}
 {id=WavePlayer_218, properties {soundWaveAssetPtr=&Outlander_Foley_Stereo_001}}
 {id=WavePlayer_219, properties {soundWaveAssetPtr=&Outlander_Foley_Stereo_003}}
 {id=WavePlayer_22, properties {soundWaveAssetPtr=&Outlander_Step_001}}
 {id=WavePlayer_220, properties {soundWaveAssetPtr=&Outlander_Foley_Stereo_005}}
 {id=WavePlayer_23, properties {soundWaveAssetPtr=&Outlander_Step_003}}
 {id=WavePlayer_24, properties {soundWaveAssetPtr=&Outlander_Step_004}}
 {id=WavePlayer_25, properties {soundWaveAssetPtr=&Outlander_Step_005}}
 {id=WavePlayer_26, properties {soundWaveAssetPtr=&Outlander_Step_007}}
 {id=WavePlayer_27, properties {soundWaveAssetPtr=&Outlander_Step_012}}
 {id=WavePlayer_28, properties {soundWaveAssetPtr=&Outlander_Foley_Stereo_002}}
 {id=WavePlayer_29, properties {soundWaveAssetPtr=&Outlander_Foley_Stereo_006}}
 {id=WavePlayer_3, properties {soundWaveAssetPtr=/Game/Sounds/Fort_Step_Sounds/Human/FS_Wood_Generic_Med_002}}
 {id=WavePlayer_30, properties {soundWaveAssetPtr=&Outlander_Foley_Stereo_010}}
 {id=WavePlayer_31, properties {soundWaveAssetPtr=&Outlander_Foley_Stereo_001}}
 {id=WavePlayer_32, properties {soundWaveAssetPtr=&Outlander_Foley_Stereo_003}}
 {id=WavePlayer_33, properties {soundWaveAssetPtr=&Outlander_Foley_Stereo_005}}
 {id=WavePlayer_34, properties {soundWaveAssetPtr=/Game/Sounds/Fort_Step_Sounds/Human/FS_Wood_Generic_Squeaks_000}}
 {id=WavePlayer_35, properties {soundWaveAssetPtr=/Game/Sounds/Fort_Step_Sounds/Human/FS_Wood_Generic_Squeaks_001}}
 {id=WavePlayer_36, properties {soundWaveAssetPtr=/Game/Sounds/Fort_Step_Sounds/Human/FS_Wood_Generic_Squeaks_002}}
 {id=WavePlayer_37, properties {soundWaveAssetPtr=/Game/Sounds/Fort_Step_Sounds/Human/FS_Wood_Generic_Squeaks_003}}
 {id=WavePlayer_38, properties {soundWaveAssetPtr=/Game/Sounds/Fort_Step_Sounds/Human/FS_Wood_Generic_Squeaks_004}}
 {id=WavePlayer_39, properties {soundWaveAssetPtr=/Game/Sounds/Fort_Step_Sounds/Human/FS_Wood_Generic_Squeaks_005}}
 {id=WavePlayer_4, properties {soundWaveAssetPtr=/Game/Sounds/Fort_Step_Sounds/Human/FS_Wood_Generic_Med_003}}
 {id=WavePlayer_40, properties {soundWaveAssetPtr=/Game/Sounds/Fort_Step_Sounds/Human/FS_Wood_Generic_Squeaks_006}}
 {id=WavePlayer_41, properties {soundWaveAssetPtr=/Game/Sounds/Fort_Step_Sounds/Human/FS_Wood_Generic_Squeaks_007}}
 {id=WavePlayer_42, properties {soundWaveAssetPtr=/Game/Sounds/Fort_Step_Sounds/Human/FS_Wood_Generic_Squeaks_008}}
 {id=WavePlayer_43, properties {soundWaveAssetPtr=/Game/Sounds/Fort_Step_Sounds/Human/FS_Metal_Stair_Run_Med_Generic_000}}
 {id=WavePlayer_44, properties {soundWaveAssetPtr=/Game/Sounds/Fort_Step_Sounds/Human/FS_Metal_Stair_Run_Med_Generic_001}}
 {id=WavePlayer_45, properties {soundWaveAssetPtr=/Game/Sounds/Fort_Step_Sounds/Human/FS_Metal_Stair_Run_Med_Generic_002}}
 {id=WavePlayer_46, properties {soundWaveAssetPtr=/Game/Sounds/Fort_Step_Sounds/Human/FS_Metal_Stair_Run_Med_Generic_003}}
 {id=WavePlayer_47, properties {soundWaveAssetPtr=/Game/Sounds/Fort_Step_Sounds/Human/FS_Metal_Stair_Run_Med_Generic_004}}
 {id=WavePlayer_48, properties {soundWaveAssetPtr=/Game/Sounds/Fort_Step_Sounds/Human/FS_Metal_Stair_Run_Med_Generic_005}}
 {id=WavePlayer_49, properties {soundWaveAssetPtr=/Game/Sounds/Fort_Step_Sounds/Human/FS_Metal_Stair_Run_Med_Generic_006}}
 {id=WavePlayer_5, properties {soundWaveAssetPtr=/Game/Sounds/Fort_Step_Sounds/Human/FS_Wood_Generic_Med_004}}
 {id=WavePlayer_50, properties {soundWaveAssetPtr=/Game/Sounds/Fort_Step_Sounds/Human/FS_Metal_Stair_Run_Med_Generic_007}}
 {id=WavePlayer_51, properties {soundWaveAssetPtr=/Game/Sounds/Fort_Step_Sounds/Human/FS_Metal_Solid_Run_Med_000}}
 {id=WavePlayer_52, properties {soundWaveAssetPtr=/Game/Sounds/Fort_Step_Sounds/Human/FS_Metal_Solid_Run_Med_001}}
 {id=WavePlayer_53, properties {soundWaveAssetPtr=/Game/Sounds/Fort_Step_Sounds/Human/FS_Metal_Solid_Run_Med_002}}
 {id=WavePlayer_54, properties {soundWaveAssetPtr=/Game/Sounds/Fort_Step_Sounds/Human/FS_Metal_Solid_Run_Med_003}}
 {id=WavePlayer_55, properties {soundWaveAssetPtr=/Game/Sounds/Fort_Step_Sounds/Human/FS_Metal_Solid_Run_Med_004}}
 {id=WavePlayer_56, properties {soundWaveAssetPtr=/Game/Sounds/Fort_Step_Sounds/Human/FS_Metal_Solid_Run_Med_005}}
 {id=WavePlayer_57, properties {soundWaveAssetPtr=/Game/Sounds/Fort_Step_Sounds/Human/FS_Metal_Solid_Run_Med_006}}
 {id=WavePlayer_58, properties {soundWaveAssetPtr=/Game/Sounds/Fort_Step_Sounds/Human/FS_Metal_Solid_Run_Med_007}}
 {id=WavePlayer_59, properties {soundWaveAssetPtr=/Game/Sounds/Fort_Step_Sounds/Human/FS_Metal_Solid_Run_Med_008}}
 {id=WavePlayer_6, properties {soundWaveAssetPtr=/Game/Sounds/Fort_Step_Sounds/Human/FS_Wood_Generic_Med_005}}
 {id=WavePlayer_61, properties {soundWaveAssetPtr=/Game/Sounds/Fort_Step_Sounds/Human/FS_Metal_Hollow_Run_000}}
 {id=WavePlayer_62, properties {soundWaveAssetPtr=&Outlander_Foley_Stereo_002}}
 {id=WavePlayer_63, properties {soundWaveAssetPtr=&Outlander_Foley_Stereo_006}}
 {id=WavePlayer_64, properties {soundWaveAssetPtr=&Outlander_Foley_Stereo_010}}
 {id=WavePlayer_65, properties {soundWaveAssetPtr=&Outlander_Foley_Stereo_001}}
 {id=WavePlayer_66, properties {soundWaveAssetPtr=&Outlander_Foley_Stereo_003}}
 {id=WavePlayer_67, properties {soundWaveAssetPtr=&Outlander_Foley_Stereo_005}}
 {id=WavePlayer_68, properties {soundWaveAssetPtr=/Game/Sounds/Fort_Step_Sounds/Human/FS_Metal_Hollow_Run_001}}
 {id=WavePlayer_69, properties {soundWaveAssetPtr=/Game/Sounds/Fort_Step_Sounds/Human/FS_Metal_Hollow_Run_002}}
 {id=WavePlayer_7, properties {soundWaveAssetPtr=/Game/Sounds/Fort_Step_Sounds/Human/FS_Wood_Generic_Med_006}}
 {id=WavePlayer_70, properties {soundWaveAssetPtr=/Game/Sounds/Fort_Step_Sounds/Human/FS_Metal_Hollow_Run_003}}
 {id=WavePlayer_71, properties {soundWaveAssetPtr=/Game/Sounds/Fort_Step_Sounds/Human/FS_Metal_Hollow_Run_004}}
 {id=WavePlayer_72, properties {soundWaveAssetPtr=/Game/Sounds/Fort_Step_Sounds/Human/FS_Metal_Hollow_Run_005}}
 {id=WavePlayer_73, properties {soundWaveAssetPtr=/Game/Sounds/Fort_Step_Sounds/Human/FS_Metal_Hollow_Run_006}}
 {id=WavePlayer_74, properties {soundWaveAssetPtr=/Game/Sounds/Fort_Step_Sounds/Human/FS_Metal_Walk_Med_Generic_000}}
 {id=WavePlayer_75, properties {soundWaveAssetPtr=/Game/Sounds/Fort_Step_Sounds/Human/FS_Metal_Walk_Med_Generic_001}}
 {id=WavePlayer_76, properties {soundWaveAssetPtr=/Game/Sounds/Fort_Step_Sounds/Human/FS_Metal_Walk_Med_Generic_002}}
 {id=WavePlayer_77, properties {soundWaveAssetPtr=/Game/Sounds/Fort_Step_Sounds/Human/FS_Metal_Walk_Med_Generic_003}}
 {id=WavePlayer_78, properties {soundWaveAssetPtr=/Game/Sounds/Fort_Step_Sounds/Human/FS_Metal_Walk_Med_Generic_004}}
 {id=WavePlayer_79, properties {soundWaveAssetPtr=/Game/Sounds/Fort_Step_Sounds/Human/FS_Metal_Walk_Med_Generic_005}}
 {id=WavePlayer_8, properties {soundWaveAssetPtr=/Game/Sounds/Fort_Step_Sounds/Human/FS_Wood_Generic_Med_007}}
 {id=WavePlayer_80, properties {soundWaveAssetPtr=/Game/Sounds/Fort_Step_Sounds/Human/FS_Metal_Walk_Med_Generic_006}}
 {id=WavePlayer_81, properties {soundWaveAssetPtr=/Game/Sounds/Fort_Step_Sounds/Human/FS_Metal_Walk_Med_Generic_007}}
 {id=WavePlayer_85, properties {soundWaveAssetPtr=&footsteps_dirt_trail_07}}
 {id=WavePlayer_86, properties {soundWaveAssetPtr=&footsteps_dirt_trail_05}}
 {id=WavePlayer_87, properties {soundWaveAssetPtr=&footsteps_dirt_trail_03}}
 {id=WavePlayer_88, properties {soundWaveAssetPtr=&footsteps_dirt_trail_02}}
 {id=WavePlayer_89, properties {soundWaveAssetPtr=&footsteps_dirt_trail_01}}
 {id=WavePlayer_9, properties {soundWaveAssetPtr=/Game/Sounds/Fort_Step_Sounds/Human/FS_Wood_Generic_MedHvy_000}}
 {id=WavePlayer_90, properties {soundWaveAssetPtr=&Outlander_Step_001}}
 {id=WavePlayer_91, properties {soundWaveAssetPtr=&Outlander_Step_003}}
 {id=WavePlayer_92, properties {soundWaveAssetPtr=&Outlander_Step_004}}
 {id=WavePlayer_93, properties {soundWaveAssetPtr=&Outlander_Step_005}}
 {id=WavePlayer_94, properties {soundWaveAssetPtr=&Outlander_Step_007}}
 {id=WavePlayer_95, properties {soundWaveAssetPtr=&Outlander_Step_012}}
 {id=WavePlayer_96, properties {soundWaveAssetPtr=&Outlander_Foley_Stereo_002}}
 {id=WavePlayer_97, properties {soundWaveAssetPtr=&Outlander_Foley_Stereo_006}}
 {id=WavePlayer_98, properties {soundWaveAssetPtr=&Outlander_Foley_Stereo_010}}
 {id=WavePlayer_99, properties {soundWaveAssetPtr=&Outlander_Foley_Stereo_001}}
]
properties {
 attenuationSettings=/Game/Sounds/Attenuation/Fort_Player_Footsteps_Att
 bOverrideConcurrency=true
 maxDistance=3500
 pitchMultiplier=1.3
 priority=0.5
 soundClassObject=&Fort_Player_Foley
 volumeMultiplier=0.7
 concurrencyOverrides {maxCount=4, resolutionRule=StopOldest}
}
