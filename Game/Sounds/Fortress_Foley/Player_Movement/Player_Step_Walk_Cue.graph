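# Player walk footstep cue that picks a surface variant through the FS_Surfaces integer switch: wood, metal, metal stairs, or generic Outlander steps for the other slots.
# Each variant mixes randomized, modulated step samples with a quiet, briefly delayed Outlander foley layer.
# Wood adds a medium-heavy step layer and an occasional squeak via a weighted empty random branch.
# Footstep attenuation applies, and concurrency caps at four voices, stopping the oldest.
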SoundCue /Game/Sounds/Fortress_Foley/Player_Movement/Player_Step_Walk_Cue
root=Switch_0
nodes [
 {id=Delay_1, children [Modulator_2], properties {delayMax=0.2}}
 {id=Delay_10, children [Modulator_1], properties {delayMax=0.2}}
 {id=Delay_11, children [Modulator_22], properties {delayMax=0.2}}
 {id=Delay_4, children [Modulator_8], properties {delayMax=0.2}}
 {id=Delay_7, children [Modulator_14], properties {delayMax=0.2}}
 {id=Mixer_1, children [Modulator_3,Delay_1], properties {inputVolume [1,0.15]}}
 {id=Mixer_10, children [Modulator_4,Delay_10], properties {inputVolume [1,0.15]}}
 {id=Mixer_11, children [Modulator_23,Delay_11], properties {inputVolume [1,0.15]}}
 {id=Mixer_4, children [Modulator_9,Delay_4,Random_10,Random_11], properties {inputVolume [1,0.15,1,1]}}
 {id=Mixer_7, children [Modulator_15,Delay_7], properties {inputVolume [1,0.15]}}
 {id=Modulator_1, children [Random_21], properties {pitchMax=1.5, pitchMin=1.3, volumeMax=1, volumeMin=0.3}}
 {id=Modulator_14, children [Random_15], properties {pitchMax=1.5, pitchMin=1.3, volumeMax=1, volumeMin=0.3}}
 {id=Modulator_15, children [Random_14], properties {pitchMax=1.6, pitchMin=1.2, volumeMax=1, volumeMin=0.7}}
 {id=Modulator_2, children [Random_3], properties {pitchMax=1.5, pitchMin=1.3, volumeMax=1, volumeMin=0.3}}
 {id=Modulator_22, children [Random_23], properties {pitchMax=1.5, pitchMin=1.3, volumeMax=1, volumeMin=0.3}}
 {id=Modulator_23, children [Random_0], properties {pitchMax=1, volumeMax=1, volumeMin=0.7}}
 {id=Modulator_3, children [Random_2], properties {pitchMax=1.6, pitchMin=1.2, volumeMax=1, volumeMin=0.7}}
 {id=Modulator_4, children [Random_4], properties {pitchMax=1, pitchMin=0.94, volumeMax=1, volumeMin=0.3}}
 {id=Modulator_8, children [Random_9], properties {pitchMax=1.5, pitchMin=1.3, volumeMax=1, volumeMin=0.3}}
 {id=Modulator_9, children [Random_6], properties {pitchMax=1.6, pitchMin=1.2, volumeMax=1, volumeMin=0.7}}
 {
  id=Random_0
  children [WavePlayer_0,WavePlayer_1,WavePlayer_2,WavePlayer_3,WavePlayer_4,WavePlayer_5,WavePlayer_6,WavePlayer_7]
  properties {weights [1,1,1,1,1,1,1,1]}
 }
 {id=Random_10, children [Random_7,null], properties {bRandomizeWithoutReplacement=false, weights [0.5,1]}}
 {
  id=Random_11
  children [WavePlayer_65, WavePlayer_66, WavePlayer_67, WavePlayer_68, WavePlayer_69, WavePlayer_70, WavePlayer_71]
  properties {weights [1,1,1,1,1,1,1]}
 }
 {
  id=Random_14
  children [WavePlayer_84, WavePlayer_85, WavePlayer_86, WavePlayer_87, WavePlayer_88, WavePlayer_89]
  properties {weights [1,1,1,1,1,1]}
 }
 {
  id=Random_15
  children [WavePlayer_90, WavePlayer_91, WavePlayer_92, WavePlayer_93, WavePlayer_94, WavePlayer_95]
  properties {weights [1,1,1,1,1,1]}
 }
 {
  id=Random_2
  children [WavePlayer_12, WavePlayer_13, WavePlayer_14, WavePlayer_15, WavePlayer_16, WavePlayer_17]
  properties {weights [1,1,1,1,1,1]}
 }
 {
  id=Random_21
  children [WavePlayer_126, WavePlayer_127, WavePlayer_128, WavePlayer_129, WavePlayer_130, WavePlayer_131]
  properties {weights [1,1,1,1,1,1]}
 }
 {
  id=Random_23
  children [WavePlayer_138, WavePlayer_139, WavePlayer_140, WavePlayer_141, WavePlayer_142, WavePlayer_143]
  properties {weights [1,1,1,1,1,1]}
 }
 {
  id=Random_3
  children [WavePlayer_18, WavePlayer_19, WavePlayer_20, WavePlayer_21, WavePlayer_22, WavePlayer_23]
  properties {weights [1,1,1,1,1,1]}
 }
 {
  id=Random_4
  children [WavePlayer_30, WavePlayer_31, WavePlayer_32, WavePlayer_33, WavePlayer_34, WavePlayer_35, WavePlayer_48, WavePlayer_49, WavePlayer_50]
  properties {weights [1,1,1,1,1,1,1,1,1]}
 }
 {
  id=Random_6
  children [WavePlayer_36, WavePlayer_37, WavePlayer_38, WavePlayer_39, WavePlayer_40, WavePlayer_41, WavePlayer_42, WavePlayer_43]
  properties {weights [1,1,1,1,1,1,1,1]}
 }
 {
  id=Random_7
  children [WavePlayer_44, WavePlayer_45, WavePlayer_46, WavePlayer_47, WavePlayer_60, WavePlayer_61, WavePlayer_62, WavePlayer_63, WavePlayer_64]
  properties {weights [1,1,1,1,1,1,1,1,1]}
 }
 {
  id=Random_9
  children [WavePlayer_54, WavePlayer_55, WavePlayer_56, WavePlayer_57, WavePlayer_58, WavePlayer_59]
  properties {weights [1,1,1,1,1,1]}
 }
 {id=Switch_0, children [Mixer_1,Mixer_1,Mixer_4,Mixer_7,Mixer_10,Mixer_11], properties {intParameterName=FS_Surfaces}}
 {id=WavePlayer_0, properties {soundWaveAssetPtr=/Game/Sounds/Fort_Step_Sounds/Human/FS_Metal_Stair_Walk_Med_Generic_000}}
 {id=WavePlayer_1, properties {soundWaveAssetPtr=/Game/Sounds/Fort_Step_Sounds/Human/FS_Metal_Stair_Walk_Med_Generic_001}}
 {id=WavePlayer_12, properties {soundWaveAssetPtr=/Game/Sounds/Fortress_Foley/Outlander/Outlander_Step_001}}
 {id=WavePlayer_126, properties {soundWaveAssetPtr=/Game/Sounds/Fortress_Foley/Outlander/Outlander_Foley_Stereo_002}}
 {id=WavePlayer_127, properties {soundWaveAssetPtr=/Game/Sounds/Fortress_Foley/Outlander/Outlander_Foley_Stereo_006}}
 {id=WavePlayer_128, properties {soundWaveAssetPtr=/Game/Sounds/Fortress_Foley/Outlander/Outlander_Foley_Stereo_010}}
 {id=WavePlayer_129, properties {soundWaveAssetPtr=/Game/Sounds/Fortress_Foley/Outlander/Outlander_Foley_Stereo_001}}
 {id=WavePlayer_13, properties {soundWaveAssetPtr=/Game/Sounds/Fortress_Foley/Outlander/Outlander_Step_003}}
 {id=WavePlayer_130, properties {soundWaveAssetPtr=/Game/Sounds/Fortress_Foley/Outlander/Outlander_Foley_Stereo_003}}
 {id=WavePlayer_131, properties {soundWaveAssetPtr=/Game/Sounds/Fortress_Foley/Outlander/Outlander_Foley_Stereo_005}}
 {id=WavePlayer_138, properties {soundWaveAssetPtr=&Outlander_Foley_Stereo_002}}
 {id=WavePlayer_139, properties {soundWaveAssetPtr=&Outlander_Foley_Stereo_006}}
 {id=WavePlayer_14, properties {soundWaveAssetPtr=/Game/Sounds/Fortress_Foley/Outlander/Outlander_Step_004}}
 {id=WavePlayer_140, properties {soundWaveAssetPtr=&Outlander_Foley_Stereo_010}}
 {id=WavePlayer_141, properties {soundWaveAssetPtr=&Outlander_Foley_Stereo_001}}
 {id=WavePlayer_142, properties {soundWaveAssetPtr=&Outlander_Foley_Stereo_003}}
 {id=WavePlayer_143, properties {soundWaveAssetPtr=&Outlander_Foley_Stereo_005}}
 {id=WavePlayer_15, properties {soundWaveAssetPtr=/Game/Sounds/Fortress_Foley/Outlander/Outlander_Step_005}}
 {id=WavePlayer_16, properties {soundWaveAssetPtr=/Game/Sounds/Fortress_Foley/Outlander/Outlander_Step_007}}
 {id=WavePlayer_17, properties {soundWaveAssetPtr=/Game/Sounds/Fortress_Foley/Outlander/Outlander_Step_012}}
 {id=WavePlayer_18, properties {soundWaveAssetPtr=&Outlander_Foley_Stereo_002}}
 {id=WavePlayer_19, properties {soundWaveAssetPtr=&Outlander_Foley_Stereo_006}}
 {id=WavePlayer_2, properties {soundWaveAssetPtr=/Game/Sounds/Fort_Step_Sounds/Human/FS_Metal_Stair_Walk_Med_Generic_002}}
 {id=WavePlayer_20, properties {soundWaveAssetPtr=&Outlander_Foley_Stereo_010}}
 {id=WavePlayer_21, properties {soundWaveAssetPtr=&Outlander_Foley_Stereo_001}}
 {id=WavePlayer_22, properties {soundWaveAssetPtr=&Outlander_Foley_Stereo_003}}
 {id=WavePlayer_23, properties {soundWaveAssetPtr=&Outlander_Foley_Stereo_005}}
 {id=WavePlayer_3, properties {soundWaveAssetPtr=/Game/Sounds/Fort_Step_Sounds/Human/FS_Metal_Stair_Walk_Med_Generic_003}}
 {id=WavePlayer_30, properties {soundWaveAssetPtr=/Game/Sounds/Fort_Step_Sounds/Human/FS_Metal_Walk_Med_Generic_000}}
 {id=WavePlayer_31, properties {soundWaveAssetPtr=/Game/Sounds/Fort_Step_Sounds/Human/FS_Metal_Walk_Med_Generic_001}}
 {id=WavePlayer_32, properties {soundWaveAssetPtr=/Game/Sounds/Fort_Step_Sounds/Human/FS_Metal_Walk_Med_Generic_002}}
 {id=WavePlayer_33, properties {soundWaveAssetPtr=/Game/Sounds/Fort_Step_Sounds/Human/FS_Metal_Walk_Med_Generic_003}}
 {id=WavePlayer_34, properties {soundWaveAssetPtr=/Game/Sounds/Fort_Step_Sounds/Human/FS_Metal_Walk_Med_Generic_004}}
 {id=WavePlayer_35, properties {soundWaveAssetPtr=/Game/Sounds/Fort_Step_Sounds/Human/FS_Metal_Walk_Med_Generic_005}}
 {id=WavePlayer_36, properties {soundWaveAssetPtr=/Game/Sounds/Fort_Step_Sounds/Human/FS_Wood_Walk_Med_Generic_000}}
 {id=WavePlayer_37, properties {soundWaveAssetPtr=/Game/Sounds/Fort_Step_Sounds/Human/FS_Wood_Walk_Med_Generic_001}}
 {id=WavePlayer_38, properties {soundWaveAssetPtr=/Game/Sounds/Fort_Step_Sounds/Human/FS_Wood_Walk_Med_Generic_002}}
 {id=WavePlayer_39, properties {soundWaveAssetPtr=/Game/Sounds/Fort_Step_Sounds/Human/FS_Wood_Walk_Med_Generic_003}}
 {id=WavePlayer_4, properties {soundWaveAssetPtr=/Game/Sounds/Fort_Step_Sounds/Human/FS_Metal_Stair_Walk_Med_Generic_004}}
 {id=WavePlayer_40, properties {soundWaveAssetPtr=/Game/Sounds/Fort_Step_Sounds/Human/FS_Wood_Walk_Med_Generic_004}}
 {id=WavePlayer_41, properties {soundWaveAssetPtr=/Game/Sounds/Fort_Step_Sounds/Human/FS_Wood_Walk_Med_Generic_005}}
 {id=WavePlayer_42, properties {soundWaveAssetPtr=/Game/Sounds/Fort_Step_Sounds/Human/FS_Wood_Walk_Med_Generic_006}}
 {id=WavePlayer_43, properties {soundWaveAssetPtr=/Game/Sounds/Fort_Step_Sounds/Human/FS_Wood_Walk_Med_Generic_007}}
 {id=WavePlayer_44, properties {soundWaveAssetPtr=/Game/Sounds/Fort_Step_Sounds/Human/FS_Wood_Generic_Squeaks_000}}
 {id=WavePlayer_45, properties {soundWaveAssetPtr=/Game/Sounds/Fort_Step_Sounds/Human/FS_Wood_Generic_Squeaks_001}}
 {id=WavePlayer_46, properties {soundWaveAssetPtr=/Game/Sounds/Fort_Step_Sounds/Human/FS_Wood_Generic_Squeaks_002}}
 {id=WavePlayer_47, properties {soundWaveAssetPtr=/Game/Sounds/Fort_Step_Sounds/Human/FS_Wood_Generic_Squeaks_003}}
 {id=WavePlayer_48, properties {soundWaveAssetPtr=/Game/Sounds/Fort_Step_Sounds/Human/FS_Metal_Walk_Med_Generic_006}}
 {id=WavePlayer_49, properties {soundWaveAssetPtr=/Game/Sounds/Fort_Step_Sounds/Human/FS_Metal_Walk_Med_Generic_007}}
 {id=WavePlayer_5, properties {soundWaveAssetPtr=/Game/Sounds/Fort_Step_Sounds/Human/FS_Metal_Stair_Walk_Med_Generic_005}}
 {id=WavePlayer_50, properties {soundWaveAssetPtr=/Game/Sounds/Fort_Step_Sounds/Human/FS_Metal_Walk_Med_Generic_008}}
 {id=WavePlayer_54, properties {soundWaveAssetPtr=&Outlander_Foley_Stereo_002}}
 {id=WavePlayer_55, properties {soundWaveAssetPtr=&Outlander_Foley_Stereo_006}}
 {id=WavePlayer_56, properties {soundWaveAssetPtr=&Outlander_Foley_Stereo_010}}
 {id=WavePlayer_57, properties {soundWaveAssetPtr=&Outlander_Foley_Stereo_001}}
 {id=WavePlayer_58, properties {soundWaveAssetPtr=&Outlander_Foley_Stereo_003}}
 {id=WavePlayer_59, properties {soundWaveAssetPtr=&Outlander_Foley_Stereo_005}}
 {id=WavePlayer_6, properties {soundWaveAssetPtr=/Game/Sounds/Fort_Step_Sounds/Human/FS_Metal_Stair_Walk_Med_Generic_006}}
 {id=WavePlayer_60, properties {soundWaveAssetPtr=/Game/Sounds/Fort_Step_Sounds/Human/FS_Wood_Generic_Squeaks_004}}
 {id=WavePlayer_61, properties {soundWaveAssetPtr=/Game/Sounds/Fort_Step_Sounds/Human/FS_Wood_Generic_Squeaks_005}}
 {id=WavePlayer_62, properties {soundWaveAssetPtr=/Game/Sounds/Fort_Step_Sounds/Human/FS_Wood_Generic_Squeaks_006}}
 {id=WavePlayer_63, properties {soundWaveAssetPtr=/Game/Sounds/Fort_Step_Sounds/Human/FS_Wood_Generic_Squeaks_007}}
 {id=WavePlayer_64, properties {soundWaveAssetPtr=/Game/Sounds/Fort_Step_Sounds/Human/FS_Wood_Generic_Squeaks_008}}
 {id=WavePlayer_65, properties {soundWaveAssetPtr=/Game/Sounds/Fort_Step_Sounds/Human/FS_Wood_Walk_MedHvy_Generic_000}}
 {id=WavePlayer_66, properties {soundWaveAssetPtr=/Game/Sounds/Fort_Step_Sounds/Human/FS_Wood_Walk_MedHvy_Generic_001}}
 {id=WavePlayer_67, properties {soundWaveAssetPtr=/Game/Sounds/Fort_Step_Sounds/Human/FS_Wood_Walk_MedHvy_Generic_002}}
 {id=WavePlayer_68, properties {soundWaveAssetPtr=/Game/Sounds/Fort_Step_Sounds/Human/FS_Wood_Walk_MedHvy_Generic_003}}
 {id=WavePlayer_69, properties {soundWaveAssetPtr=/Game/Sounds/Fort_Step_Sounds/Human/FS_Wood_Walk_MedHvy_Generic_004}}
 {id=WavePlayer_7, properties {soundWaveAssetPtr=/Game/Sounds/Fort_Step_Sounds/Human/FS_Metal_Stair_Walk_Med_Generic_007}}
 {id=WavePlayer_70, properties {soundWaveAssetPtr=/Game/Sounds/Fort_Step_Sounds/Human/FS_Wood_Walk_MedHvy_Generic_005}}
 {id=WavePlayer_71, properties {soundWaveAssetPtr=/Game/Sounds/Fort_Step_Sounds/Human/FS_Wood_Walk_MedHvy_Generic_006}}
 {id=WavePlayer_84, properties {soundWaveAssetPtr=&Outlander_Step_001}}
 {id=WavePlayer_85, properties {soundWaveAssetPtr=&Outlander_Step_003}}
 {id=WavePlayer_86, properties {soundWaveAssetPtr=&Outlander_Step_004}}
 {id=WavePlayer_87, properties {soundWaveAssetPtr=&Outlander_Step_005}}
 {id=WavePlayer_88, properties {soundWaveAssetPtr=&Outlander_Step_007}}
 {id=WavePlayer_89, properties {soundWaveAssetPtr=&Outlander_Step_012}}
 {id=WavePlayer_90, properties {soundWaveAssetPtr=&Outlander_Foley_Stereo_002}}
 {id=WavePlayer_91, properties {soundWaveAssetPtr=&Outlander_Foley_Stereo_006}}
 {id=WavePlayer_92, properties {soundWaveAssetPtr=&Outlander_Foley_Stereo_010}}
 {id=WavePlayer_93, properties {soundWaveAssetPtr=&Outlander_Foley_Stereo_001}}
 {id=WavePlayer_94, properties {soundWaveAssetPtr=&Outlander_Foley_Stereo_003}}
 {id=WavePlayer_95, properties {soundWaveAssetPtr=&Outlander_Foley_Stereo_005}}
]
properties {
 attenuationSettings=/Game/Sounds/Attenuation/Fort_Player_Footsteps_Att
 bOverrideConcurrency=true
 maxDistance=3500
 priority=0.5
 soundClassObject=&Fort_Player_Foley
 volumeMultiplier=0.5
 concurrencyOverrides {maxCount=4, resolutionRule=StopOldest}
}
